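# Unlit masked material that, by its name, hides gadget prop meshes by making them invisible and crunching their vertices.
# A zero constant fits an opacity mask, and camera-relative object position minus world position gives an offset that would collapse each vertex onto the object's origin, though output connections are not shown.
# It supports skeletal meshes and morph targets, and disables ray-traced shadows.
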Material /BRCosmetics/Animation/Props/Gadgets_PixieParts/Materials/M_Invisible_VertCrunched
graph {
 properties {
  bAutomaticallySetUsageInEditor=false
  bCastRayTracedShadows=false
  bUsedWithMorphTargets=true
  bUsedWithSkeletalMesh=true
  blendMode=BLEND_Masked
  shadingModel=MSM_Unlit
  usedShadingModels=Unlit
 }
 nodes [
  Constant()
  Subtract(A=ObjectPositionWS() {originType=CameraRelative}, B=WorldPosition() {worldPositionShaderOffset=WPT_CameraRelative}.XYZ)
 ]
}
